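; Behavior tree injection evaluator that gates a Fortnite AI bot's Team Spray behavior on the bot pawn's missing health or shield.
; Starting requires the shield or health deficit to reach its scalable-float threshold, without consulting the parent check.
; Continuing requires the parent check to pass and either stat to remain below maximum, creating hysteresis so the behavior can run until both are full.
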(fn EvaluateOngoingConditions
  (params (OwnerComp :type Behavior%20Tree%20Component%20Object%20Reference))
  (bind _returnvalue (|CallParentFunction OwnerComp))
  (if _returnvalue
    (bind _asfort_athena_aibot_controller (Utilities|Casting|CastToFortAthenaAIBotController (Components|GetOwner OwnerComp))
      (:then
        (bind _asfort_pawn (Utilities|Casting|CastToFortPawn (Pawn|GetControlledPawn _asfort_athena_aibot_controller))
          (:then
            (return (or (< (FortPawn|GetShield _asfort_pawn) (FortPawn|GetMaxShield _asfort_pawn)) (< (FortPawn|GetHealth _asfort_pawn) (FortPawn|GetMaxHealth _asfort_pawn)))))
          (:CastFailed
            _)))
      (:CastFailed
        _))
    (return false)
    (else
      _)))

(fn EvaluateStartingConditions
  (params (OwnerComp :type Behavior%20Tree%20Component%20Object%20Reference))
  (bind _asfort_athena_aibot_controller (Utilities|Casting|CastToFortAthenaAIBotController (Components|GetOwner OwnerComp))
    (:then
      (bind _asfort_pawn (Utilities|Casting|CastToFortPawn (Pawn|GetControlledPawn _asfort_athena_aibot_controller))
        (:then
          (return (or (>= (- (FortPawn|GetMaxShield _asfort_pawn) (FortPawn|GetShield _asfort_pawn)) (ScalableFloat|GetValueatLevel (Variables|Default|GetShieldThreshold))) (>= (- (FortPawn|GetMaxHealth _asfort_pawn) (FortPawn|GetHealth _asfort_pawn)) (ScalableFloat|GetValueatLevel (Variables|Default|GetHealthThreshold))))))
        (:CastFailed
          _)))
    (:CastFailed
      _))
  (return false))
